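# Material function that applies a SuperLevel cosmetic style (T02 S33) to the RoseDepth_Seed character body's material attributes.
# Two mask textures split switchable areas such as outline, secondary, hair and metal or shirt.
# Preskinned-position UVs drive four time-pulsed line layers and a flowmap-distorted "candy" gradient, shaded with Fresnel and anime-style specular terms.
# Results override base color, emissive, roughness and metallic, or only the first two for IsFlyHair.
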MaterialFunction /BRCosmetics/Characters/Player/Female/Medium/Bodies/F_MED_RoseDepth_Seed/Skins/SuperLevel/T02/MF_SuperLevel_T02_S33
graph {nodes [
  n1 = VectorParameter() {group=SuperLevel, parameterName=DistortPanning, defaultValue=<rgba 0 0.1 0 0>}
  n2 = TextureObjectParameter() {
   group=SuperLevel
   parameterName=Flowmap01
   samplerType=SAMPLERTYPE_LinearColor
   texture=/Game/Global/Textures/Default/Flowmaps/T_TorusFlowmap
  }
  n3 = TextureSample(UVs=Panner(Coordinate=NamedRerouteUsage(), Speed=ComponentMask(None=@n1.RGB) {g=true, r=true}), Tex=@n2) {samplerType=SAMPLERTYPE_LinearColor, texture=&T_TorusFlowmap}
  n4 = Multiply(A=ComponentMask(None=@n3.RGB) {g=true, r=true}, B=ScalarParameter() {defaultValue=0.1, group=SuperLevel, parameterName=Distort_Intensity}) {constB=0.1}
  n5 = MaterialFunctionCall!(Angle=ScalarParameter() {defaultValue=-0.12, group=SuperLevel, parameterName=PatternRotate03}, UV/2DVector=Add(A=@n4, B=NamedRerouteUsage())) {materialFunction=/Game/Characters/Player/Male/Medium/Bodies/M_MED_Faux/Materials/MF_2DRotate}
  Add(A=@n4, B=@n5.Result)
  n6 = NamedRerouteUsage()
  n7 = LinearInterpolate(
   A=ScalarParameter() {defaultValue=0.02, group=SuperLevel, parameterName=MinEM}
   B=ScalarParameter() {defaultValue=0.5, group=SuperLevel, parameterName=MaxEM}
   Alpha=@n6
  ) {constA=0.4}
  Add(A=@n7, B=@n6)
  AtmosphericLightVector()
  Constant()
  Custom() {code="return View.DirectionalLightDirection;", description=DirectionalLightVector, inputs []}
  n8 = FunctionInput() {inputName=Material, inputType=FunctionInput_MaterialAttributes}
  n9 = NamedRerouteUsage()
  n10 = NamedRerouteUsage()
  n11 = GetMaterialAttributes(None=@n8) {attributeGetTypes [
    #1
    #2
    #3
    #4
    #5
   ]}
  n12 = NamedRerouteDeclaration(None=@n11.Roughness) {name=RoughnessIn, nodeColor=<rgba 0 0.4844 1 1>}
  n13 = NamedRerouteUsage()
  n14 = NamedRerouteUsage()
  n15 = Add(A=@n11.Metallic, B=NamedRerouteUsage())
  n16 = NamedRerouteUsage()
  n17 = SetMaterialAttributes!(
   MaterialAttributes=@n11.MaterialAttributes
   BaseColor=NamedRerouteUsage()
   EmissiveColor=NamedRerouteUsage()
   Roughness=StaticSwitchParameter(True=LinearInterpolate(A=@n12, B=@n13, Alpha=@n14), False=@n12) {defaultValue=true, group=SuperLevel_Control, parameterName=UseSecondaryArea}
   Metallic=StaticSwitchParameter(True=LinearInterpolate(A=@n15, Alpha=@n16), False=@n15) {group=SuperLevel_Control, parameterName=UseMetalArea}
  ) {attributeSetTypes [
    #1
    #2
    #3
    #5
   ]}
  FunctionOutput(None=StaticSwitchParameter(
    True=SetMaterialAttributes!(MaterialAttributes=@n8, BaseColor=@n9, EmissiveColor=@n10) {attributeSetTypes [#1, #2]}
    False=@n17
   ) {group=SuperLevel_Control, parameterName=IsFlyHair})
  MaterialFunctionCall!(ObjectPivotLocation, ObjectPositionOffset=Constant3Vector() {constant=<rgba 0 0 110 0>}) {materialFunction=/Game/Characters/Player/Common/MaterialFunctions/Superlevel/ImprovedStabilizedScreenSpaceUVs}
  Multiply(A=@n6) {constB=2}
  n18 = Desaturation(None=NamedRerouteUsage(), Fraction=Constant() {r=1})
  n19 = NamedRerouteUsage()
  Multiply(A=@n18, B=@n19)
  n20 = Fresnel() {baseReflectFraction=0, exponent=1}
  n21 = Sine(None=Multiply(A=@n20, B=ScalarParameter() {defaultValue=0.5, group=SuperLevel, parameterName=OutlineSine}))
  n22 = VectorParameter() {group=SuperLevel, parameterName=OutlineColor_EM, channelNames {a=Intensity, b="", g="", r=""}, defaultValue=<rgba 1 0 0 10>}
  n23 = Multiply(A=Multiply(A=@n20, B=ConstantBiasScale(None=@n21)), B=Multiply(A=@n22, B=@n22))
  NamedRerouteDeclaration(None=@n23) {name=OutLineColor, nodeColor=<rgba 0 0.5312 1 1>}
  NamedRerouteDeclaration(None=FunctionInput() {inputName=PreskinnedPos, sortPriority=1}) {name=PreSkinnedPos, nodeColor=<rgba 1 0 0.4922 1>}
  n24 = VectorParameter() {group=SuperLevel_Hair, parameterName=HairColorDark, defaultValue=<rgba 0.0625 0.0043 0.0043 1>}
  n25 = VectorParameter() {group=SuperLevel_Hair, parameterName=HairColor, defaultValue=<rgba 1 1 1 1>}
  n26 = NamedRerouteUsage()
  n27 = Constant() {r=1}
  n28 = Saturate(None=LinearInterpolate(A=Multiply(A=@n24.RGB, B=@n24.RGB), B=Multiply(A=@n25.RGB, B=@n25.RGB), Alpha=Desaturation(None=@n26, Fraction=@n27)))
  NamedRerouteDeclaration(None=@n28) {name=HairColor_Adj, nodeColor=<rgba 0 1 0.4609 1>}
  NamedRerouteDeclaration(None=FunctionInput() {inputName=PreskinnedLocalNormal, sortPriority=2}) {name=PreSkinnedNormal, nodeColor=<rgba 1 0 0.4922 1>}
  n29 = VectorParameter() {group=SuperLevel_Hair, parameterName=HairColor_EM, defaultValue=<rgba 1 1 1 1>}
  n30 = Multiply(A=@n29.RGB, B=@n29.RGB)
  NamedRerouteDeclaration(None=@n30) {name=HairEM_Adj, nodeColor=<rgba 0.0312 1 0 1>}
  n31 = TextureSampleParameter2D() {
   group=SuperLevel
   parameterName=SuperLevel_Mask02
   samplerType=SAMPLERTYPE_Masks
   texture=/BRCosmetics/Characters/Player/Female/Medium/Bodies/F_MED_RoseDepth_Seed/Textures/T_RoseDepth_Seed_Body_SL_FX2
  }
  n32 = Multiply(A=@n31.R, B=NamedRerouteUsage())
  n33 = NamedRerouteDeclaration(None=@n31.G) {name=G_Channel, nodeColor=<rgba 0 0.1562 1 1>}
  n34 = Saturate(None=Subtract(A=@n33, B=@n31.R))
  n35 = NamedRerouteUsage()
  NamedRerouteDeclaration(None=StaticSwitchParameter(True=@n32, False=StaticSwitchParameter(True=@n34, False=@n35) {group=SuperLevel_Control, parameterName=UseYellowAsMetal}) {group=SuperLevel_Control, parameterName=UseYellowAsShirt}) {name=DefaultEMArea, nodeColor=<rgba 0 1 0.6484 1>}
  n36 = NamedRerouteUsage()
  NamedRerouteDeclaration(None=StaticSwitchParameter(True=@n34, False=StaticSwitchParameter(True=@n32, False=@n36) {group=SuperLevel_Control, parameterName=UseYellowAsMetal}) {group=SuperLevel_Control, parameterName=UseYellowAsShirt}) {name=MetalArea, nodeColor=<rgba 0 1 0.6484 1>}
  n37 = TextureSampleParameter2D() {
   group=SuperLevel
   parameterName=SuperLevel_Mask01
   samplerType=SAMPLERTYPE_Masks
   texture=/BRCosmetics/Characters/Player/Female/Medium/Bodies/F_MED_RoseDepth_Seed/Textures/T_RoseDepth_Seed_Body_SL_FX
  }
  n38 = NamedRerouteUsage()
  NamedRerouteDeclaration(None=StaticSwitchParameter(True=@n31.RGB, False=StaticSwitchParameter(True=@n37.RGB, False=@n38) {group=SuperLevel, parameterName=DebugSuperLevelMask01}) {group=SuperLevel, parameterName=DebugSuperLevelMask02}) {name=BaseColorOUT, nodeColor=<rgba 0.7812 1 0 1>}
  NamedRerouteDeclaration(None=@n37.G) {name=OutLineArea, nodeColor=<rgba 0 1 0.2623 1>}
  NamedRerouteDeclaration(None=@n37.R) {name=PrimaryArea, nodeColor=<rgba 1 0.5625 0 1>}
  NamedRerouteDeclaration(None=@n37.B) {name=SecondaryArea, nodeColor=<rgba 0 0.6953 1 1>}
  NamedRerouteDeclaration(None=Saturate(None=Subtract(A=@n31.R, B=@n33))) {name=TertiaryArea, nodeColor=<rgba 1 0 0.75 1>}
  n39 = VectorParameter() {group=SuperLevel, parameterName=TintColor_BaseColor, channelNames {a=Intensity, b="", g="", r=""}, defaultValue=<rgba 1 1 1 1>}
  n40 = Multiply(A=@n39, B=@n39)
  NamedRerouteDeclaration(None=@n40) {name=TintColor_BaseColor, nodeColor=<rgba 1 0 0.4688 1>}
  NamedRerouteDeclaration(None=@n11.BaseColor) {name=BaseColorIN, nodeColor=<rgba 0 1 0.5781 1>}
  NamedRerouteDeclaration(None=@n11.EmissiveColor) {name=EM_IN, nodeColor=<rgba 0.5469 1 0 1>}
  NamedRerouteDeclaration(None=@n11.Normal) {name=Normal_IN, nodeColor=<rgba 1 0 0.7734 1>}
  n41 = NamedRerouteUsage()
  n42 = Constant() {r=1}
  n43 = Saturate(None=Multiply(
    A=Desaturation(None=@n41, Fraction=@n42)
    B=VectorParameter() {group=SuperLevel_Control, parameterName=TintEM, defaultValue=<rgba 1 0 0 1>}.RGB
   ))
  n44 = StaticSwitchParameter(True=Multiply(A=@n23, B=NamedRerouteUsage()), False=Constant()) {defaultValue=true, group=SuperLevel_Control, parameterName=UseOutlineArea}
  n45 = NamedRerouteUsage()
  n46 = ScalarParameter() {defaultValue=1, group=SuperLevel, parameterName=CandyEM}
  n47 = Multiply(A=Multiply(A=@n45, B=Multiply(A=@n7, B=@n46)), B=NamedRerouteUsage())
  n48 = Add(
   A=StaticSwitchParameter(True=Add(A=@n44, B=@n47), False=@n44) {defaultValue=true, group=SuperLevel_Control, parameterName=UseSecondaryArea}
   B=NamedRerouteUsage()
  )
  n49 = StaticSwitchParameter(True=LinearInterpolate(A=@n48, B=Multiply(A=@n30, B=@n18), Alpha=@n19), False=@n48) {group=SuperLevel_Control, parameterName=UseHairArea}
  n50 = ScalarParameter() {defaultValue=1, group=SuperLevel_DefaultEM, parameterName=DefaultEM_Min}
  n51 = LinearInterpolate(A=ScalarParameter() {group=SuperLevel_DefaultEM, parameterName=DefaultEM_Less_Min}, Alpha=NamedRerouteUsage()) {constA=0.5}
  n52 = NamedRerouteUsage()
  n53 = Multiply(
   A=VectorParameter() {group=SuperLevel_DefaultEM, parameterName=DefaultEmissive, defaultValue=<rgba 1 0.3132 0.2235 1>}.RGB
   B=Multiply(A=Multiply(A=@n50, B=@n51), B=@n52)
  )
  n54 = NamedRerouteUsage()
  n55 = Desaturation(None=@n54, Fraction=Constant() {r=1})
  n56 = VectorParameter() {group=SuperLevel_Shirt, parameterName=ShirtColor, defaultValue=<rgba 1 0 0 1>}
  n57 = Multiply(A=@n56.RGB, B=@n56.RGB)
  n58 = Multiply(A=@n52, B=StaticSwitchParameter(True=Multiply(A=@n55, B=@n57), False=@n54) {group=SuperLevel_Control, parameterName=UseYellowAsShirt})
  n59 = Add(A=@n49, B=Multiply(A=Add(A=@n53, B=@n58), B=ScalarParameter() {defaultValue=1.5, group=SuperLevel_DefaultEM, parameterName=DefaultEM_Multi}))
  NamedRerouteDeclaration(None=Add(
    A=StaticSwitchParameter(True=@n43, False=@n41) {group=SuperLevel_Control, parameterName=UseTintORG_EM}
    B=StaticSwitchParameter(True=@n59, False=@n49) {group=SuperLevel_Control, parameterName=UseDefaultEmissiveArea}
   )) {name=EM_OUT, nodeColor=<rgba 0 1 0.8125 1>}
  n60 = NamedRerouteUsage()
  n61 = Constant() {r=1}
  n62 = Saturate(None=Multiply(A=@n40, B=Desaturation(None=@n60, Fraction=@n61)))
  n63 = NamedRerouteUsage()
  n64 = VectorParameter() {group=SuperLevel, parameterName=TeriaryColor_BaseColor, channelNames {a=Intensity, b="", g="", r=""}, defaultValue=<rgba 1 0.8726 0.8527 1>}
  n65 = Multiply(A=@n64, B=@n64)
  n66 = Desaturation(None=NamedRerouteUsage(), Fraction=Constant() {r=1})
  n67 = Multiply(A=@n65, B=SmoothStep(Value=ComponentMask(None=@n66) {r=true}) {constMax=0.1})
  n68 = Multiply(A=@n65, B=ScalarParameter() {defaultValue=0.5, parameterName=DirtMulti}) {constB=0.5}
  n69 = NamedRerouteUsage()
  n70 = Saturate(None=Add(A=SmoothStep(Value=@n69) {constMin=0.5}, B=NamedRerouteUsage()))
  n71 = LinearInterpolate(
   A=StaticSwitchParameter(True=LinearInterpolate(A=@n62, B=@n60, Alpha=@n63), False=@n60) {defaultValue=true, group=SuperLevel_Control, parameterName=TintBaseColor}
   B=StaticSwitchParameter(True=Saturate(None=@n67), False=LinearInterpolate(A=@n65, B=@n68, Alpha=@n70)) {group=SuperLevel_Control, parameterName=TintTeriaryColor}
   Alpha=NamedRerouteUsage()
  ) {constB=0.02}
  n72 = VectorParameter() {
   group=SuperLevel
   parameterName=OutlineColor_BaseColor_Dark
   channelNames {a=Intensity, b="", g="", r=""}
   defaultValue=<rgba 0.2674 0.0151 0 1>
  }
  n73 = VectorParameter() {group=SuperLevel, parameterName=OutlineColor_BaseColor, channelNames {a=Intensity, b="", g="", r=""}, defaultValue=<rgba 0.4653 0.0106 0 1>}
  n74 = LinearInterpolate(A=@n72, B=Multiply(A=@n73, B=@n73), Alpha=Fresnel() {baseReflectFraction=0, exponent=1})
  n75 = NamedRerouteUsage()
  n76 = LinearInterpolate(
   A=StaticSwitchParameter(True=LinearInterpolate(A=@n71, B=@n74, Alpha=@n75), False=@n71) {defaultValue=true, group=SuperLevel_Control, parameterName=UseOutlineArea}
   Alpha=NamedRerouteUsage()
  ) {constB=0.02}
  n77 = StaticSwitchParameter(True=LinearInterpolate(A=@n76, B=NamedRerouteUsage(), Alpha=NamedRerouteUsage()), False=@n76) {defaultValue=true, group=SuperLevel_Control, parameterName=UseSecondaryArea}
  n78 = StaticSwitchParameter(True=LinearInterpolate(A=@n77, B=@n28, Alpha=NamedRerouteUsage()), False=@n77) {group=SuperLevel_Control, parameterName=UseHairArea}
  n79 = VectorParameter() {group=SuperLevel_Metal, parameterName=MetalColor_Light, defaultValue=<rgba 0.8771 0.0108 0 1>}
  n80 = VectorParameter() {group=SuperLevel_Metal, parameterName=MetalColor_Dark, defaultValue=<rgba 0.1854 0 0.0307 1>}
  n81 = Power(Base=Fresnel() {baseReflectFraction=0, exponent=1}) {constExponent=0.5}
  n82 = ConstantBiasScale(None=Sine(None=Multiply(A=@n81) {constB=2}))
  n83 = StaticSwitchParameter(True=LinearInterpolate(A=@n78, B=LinearInterpolate(A=@n79.RGB, B=@n80.RGB, Alpha=@n82), Alpha=NamedRerouteUsage()), False=@n78) {group=SuperLevel_Control, parameterName=UseMetalArea}
  n84 = NamedRerouteUsage()
  n85 = NamedRerouteUsage()
  NamedRerouteDeclaration(None=StaticSwitchParameter(True=LinearInterpolate(A=@n83, B=@n84, Alpha=@n85), False=@n83) {group=SuperLevel_Control, parameterName=UseYellowAsShirt}) {name=PostSuperLevel_BaseColor, nodeColor=<rgba 1 0 0.8672 1>}
  n86 = VectorParameter() {group=SuperLevel, parameterName=BasePatternOffset, defaultValue=<rgba 0 -0.52 0 1>}
  n87 = NamedRerouteUsage()
  n88 = VectorParameter() {group=SuperLevel, parameterName=SeamOffset, defaultValue=<rgba 0 0 0 1>}
  n89 = NamedRerouteUsage()
  n90 = Add(
   A=Divide(A=ComponentMask(None=@n87) {b=true, r=true}, B=ScalarParameter() {defaultValue=70, group=SuperLevel, parameterName=TillingSize})
   B=Multiply(A=ComponentMask(None=@n88.RGB) {g=true, r=true}, B=OneMinus(None=@n89))
  )
  n91 = ScalarParameter() {defaultValue=0.85, group=SuperLevel, parameterName=UV_Tilling_Screen01}
  NamedRerouteDeclaration(None=Add(A=ComponentMask(None=@n86.RGB) {g=true, r=true}, B=Multiply(A=@n90, B=@n91))) {name=UV_Screen01, nodeColor=<rgba 0.8984 1 0 1>}
  n92 = NamedRerouteUsage()
  n93 = TextureObjectParameter() {
   group=SuperLevel
   parameterName=Line01
   samplerType=SAMPLERTYPE_LinearColor
   texture=/BRCosmetics/Characters/Player/Female/Medium/Bodies/F_MED_RoseDepth_Seed/Skins/SuperLevel/Textures/T_SuperLevel_LineAnother_T02_S33
  }
  n94 = TextureSample(UVs=@n92, Tex=@n93) {samplerSource=SSM_Wrap_WorldGroupSettings, samplerType=SAMPLERTYPE_LinearColor}
  n95 = Multiply(A=Power(Base=@n94.R) {constExponent=2.2}, B=NamedRerouteUsage())
  n96 = Multiply(A=ConstantBiasScale(None=Sine(None=@n95)), B=@n94.B)
  n97 = SmoothStep(Value=@n96) {constMin=0.5}
  n98 = ScalarParameter() {defaultValue=0.1, group=SuperLevel, parameterName=Layer1_Intensity}
  n99 = ScalarParameter() {defaultValue=0.11, group=SuperLevel, parameterName=SineSpeed}
  n100 = Time()
  n101 = Sine(None=Add(A=Multiply(A=@n99, B=@n100), B=Power(Base=@n94.G) {constExponent=2.2}))
  n102 = Multiply(A=@n96, B=Power(Base=ConstantBiasScale(None=@n101)) {constExponent=7})
  n103 = ScalarParameter() {defaultValue=1.3, group=SuperLevel, parameterName=Layer1_Intensity_Animated}
  n104 = Multiply(
   A=VectorParameter() {group=SuperLevel, parameterName=LinesTint, defaultValue=<rgba 0.3548 0.625 0.5654 1>}.RGB
   B=Add(A=Multiply(A=@n97, B=@n98), B=Multiply(A=@n102, B=@n103))
  )
  n105 = Constant() {r=0.5}
  n106 = ScalarParameter() {defaultValue=0.35, group=SuperLevel, parameterName=BlueOffset}
  n107 = TextureSample(UVs=Add(A=@n92, B=AppendVector(A=@n105, B=@n106)), Tex=@n93) {samplerSource=SSM_Wrap_WorldGroupSettings, samplerType=SAMPLERTYPE_LinearColor}
  n108 = Multiply(A=Power(Base=@n107.R) {constExponent=2.2}, B=NamedRerouteUsage())
  n109 = Multiply(A=ConstantBiasScale(None=Sine(None=@n108)), B=@n107.B)
  n110 = Multiply(A=SmoothStep(Value=@n109) {constMin=0.5}, B=ScalarParameter() {defaultValue=0.1, group=SuperLevel, parameterName=Layer2_Intensity})
  n111 = Add(A=Multiply(A=ScalarParameter() {defaultValue=0.13, group=SuperLevel, parameterName=SineSpeed02}, B=Time())) {constB=0.08}
  n112 = Power(Base=@n107.G) {constExponent=2.2}
  n113 = ConstantBiasScale(None=Sine(None=Add(A=@n111, B=@n112)))
  n114 = Multiply(
   A=Multiply(A=@n109, B=Power(Base=@n113) {constExponent=5})
   B=ScalarParameter() {defaultValue=0.85, group=SuperLevel, parameterName=Layer2_Intensity_Animated}
  )
  n115 = VectorParameter() {group=SuperLevel, parameterName=LinesTint02, defaultValue=<rgba 0.6197 0.625 0.5165 1>}
  n116 = TextureSample(UVs=NamedRerouteUsage(), Tex=@n93) {samplerSource=SSM_Wrap_WorldGroupSettings, samplerType=SAMPLERTYPE_LinearColor}
  n117 = Multiply(A=Power(Base=@n116.R) {constExponent=2.2}, B=NamedRerouteUsage())
  n118 = Multiply(A=ConstantBiasScale(None=Sine(None=@n117)), B=@n116.B)
  n119 = ScalarParameter() {defaultValue=0.2, group=SuperLevel, parameterName=SineSpeed03}
  n120 = Time()
  n121 = Add(A=Add(A=Multiply(A=@n119, B=@n120)) {constB=0.2}, B=Power(Base=@n116.G) {constExponent=2.2})
  n122 = Power(Base=ConstantBiasScale(None=Sine(None=@n121))) {constExponent=3}
  n123 = Add(
   A=Multiply(A=SmoothStep(Value=@n118) {constMin=0.5}, B=ScalarParameter() {defaultValue=0.05, group=SuperLevel, parameterName=Layer3_Intensity})
   B=Multiply(A=Multiply(A=@n118, B=@n122), B=ScalarParameter() {defaultValue=0.4, group=SuperLevel, parameterName=Layer3_Intensity_Animated})
  )
  n124 = TextureSample(UVs=NamedRerouteUsage(), Tex=@n93) {samplerSource=SSM_Wrap_WorldGroupSettings, samplerType=SAMPLERTYPE_LinearColor}
  n125 = Multiply(A=Power(Base=@n124.R) {constExponent=2.2}, B=NamedRerouteUsage())
  n126 = Multiply(A=ConstantBiasScale(None=Sine(None=@n125)), B=@n124.B)
  n127 = Multiply(A=SmoothStep(Value=@n126) {constMin=0.5}, B=ScalarParameter() {defaultValue=0.05, group=SuperLevel, parameterName=Layer4_Intensity})
  n128 = Add(A=Multiply(A=ScalarParameter() {defaultValue=0.22, group=SuperLevel, parameterName=SineSpeed04}, B=Time())) {constB=0.12}
  n129 = Power(Base=@n124.G) {constExponent=2.2}
  n130 = ConstantBiasScale(None=Sine(None=Add(A=@n128, B=@n129)))
  n131 = Multiply(
   A=Multiply(A=@n126, B=Power(Base=@n130) {constExponent=3})
   B=ScalarParameter() {defaultValue=0.1, group=SuperLevel, parameterName=Layer4_Intensity_Animated}
  )
  n132 = Add(
   A=Multiply(A=VectorParameter() {group=SuperLevel, parameterName=LinesTint01, defaultValue=<rgba 0.338 0.559 0.7292 1>}.RGB, B=Add(A=@n110, B=@n114))
   B=Add(A=Multiply(A=@n115.RGB, B=@n123), B=Add(A=@n127, B=@n131))
  )
  NamedRerouteDeclaration(None=Multiply(A=Add(A=@n104, B=@n132), B=NamedRerouteUsage())) {name=MainPattern01, nodeColor=<rgba 1 0 0.0938 1>}
  n133 = VectorParameter() {group=SuperLevel, parameterName=SecondPatternOffset}
  n134 = ScalarParameter() {defaultValue=1.9, group=SuperLevel, parameterName=UV_Tilling_Screen02}
  NamedRerouteDeclaration(None=Add(A=ComponentMask(None=@n133.RGB) {g=true, r=true}, B=Multiply(A=@n90, B=@n134))) {name=UV_Screen02, nodeColor=<rgba 0.8984 1 0 1>}
  n135 = VectorParameter() {group=SuperLevel, parameterName=ThirdPatternOffset}
  n136 = ScalarParameter() {defaultValue=0.75, group=SuperLevel, parameterName=UV_Tilling_Screen03}
  NamedRerouteDeclaration(None=Add(A=ComponentMask(None=@n135.RGB) {g=true, r=true}, B=Multiply(A=@n90, B=@n136))) {name=UV_Screen03, nodeColor=<rgba 0.8984 1 0 1>}
  NamedRerouteDeclaration(None=Distance(A=NamedRerouteUsage(), B=Constant3Vector() {constant=<rgba 0 0 1 0>})) {name=DistanceShading, nodeColor=<rgba 0 0.5312 1 1>}
  n137 = MaterialFunctionCall!(Blend Range=Constant() {r=300}, Start Offset=Constant() {r=-300}) {materialFunction=/Engine/Functions/Engine_MaterialFunctions03/Particles/Distance_Blend}
  NamedRerouteDeclaration(None=LinearInterpolate(A=ScalarParameter() {defaultValue=10, group=SuperLevel, parameterName=LineMulti_Sine}, Alpha=@n137.Result) {constB=6}) {name=Sine_HIGH, nodeColor=<rgba 0.1016 0 1 1>}
  NamedRerouteDeclaration(None=LinearInterpolate(A=ScalarParameter() {defaultValue=5, group=SuperLevel, parameterName=LineMulti_Sine_Layer3}, Alpha=@n137.Result) {constB=3}) {name=Sine_LOW, nodeColor=<rgba 0.1016 0 1 1>}
  n138 = NamedRerouteUsage()
  n139 = Saturate(None=DotProduct(A=@n138, B=Constant3Vector() {constant=<rgba 0 1 0 0>}))
  n140 = Saturate(None=DotProduct(A=@n138, B=Constant3Vector() {constant=<rgba 0 -1 0 0>}))
  NamedRerouteDeclaration(None=SmoothStep(
    Min=ScalarParameter() {group=SuperLevel, parameterName=PreskinnedClampMIN}
    Max=ScalarParameter() {defaultValue=0.7, group=SuperLevel, parameterName=PreskinnedClampMax}
    Value=Add(A=@n139, B=@n140)
   ) {constMin=0.8}) {name=PreskinnedSeam, nodeColor=<rgba 0 1 0.3438 1>}
  NamedRerouteDeclaration(None=Multiply(A=@n90, B=ScalarParameter() {defaultValue=3.5, group=SuperLevel, parameterName=UV_Tilling_Candy01})) {name=UV_Candy01, nodeColor=<rgba 1 0 0.7031 1>}
  NamedRerouteDeclaration(None=Multiply(A=@n90, B=ScalarParameter() {defaultValue=2.5, group=SuperLevel, parameterName=UV_Tilling_Candy02})) {name=UV_Candy02, nodeColor=<rgba 1 0 0.7031 1>}
  n141 = ScalarParameter() {defaultValue=0.3, group=SuperLevel, parameterName=MinRoughness}
  n142 = Divide(
   A=ComponentMask(None=NamedRerouteUsage()) {b=true, g=true}
   B=ScalarParameter() {defaultValue=70, group=SuperLevel, parameterName=TillingSize}
  )
  n143 = ScalarParameter() {defaultValue=2.5, group=SuperLevel, parameterName=UV_Tilling_Candy03}
  n144 = TextureSample(
   UVs=MaterialFunctionCall(Angle=ScalarParameter() {defaultValue=-0.15, group=SuperLevel, parameterName=PatternRotate03}, UV/2DVector=Multiply(A=@n142, B=@n143)) {materialFunction=&MF_2DRotate}.Result
   Tex=NamedRerouteUsage()
  ) {samplerSource=SSM_Wrap_WorldGroupSettings, samplerType=SAMPLERTYPE_Masks}
  n145 = ScalarParameter() {defaultValue=-0.17, group=SuperLevel, parameterName=PatternRotate01}
  n146 = NamedRerouteUsage()
  n147 = TextureObjectParameter() {
   group=SuperLevel
   parameterName=CandyTexture
   samplerType=SAMPLERTYPE_Masks
   texture=/BRCosmetics/Characters/Player/Female/Medium/Bodies/F_MED_RoseDepth_Seed/Skins/SuperLevel/Textures/T_SuperLevel_Gradient_T02_S33
  }
  n148 = TextureSample(
   UVs=Add(
    A=MaterialFunctionCall(Angle=@n145, UV/2DVector=@n146) {materialFunction=&MF_2DRotate}.Result
    B=@n4
   )
   Tex=NamedRerouteDeclaration(None=@n147) {name=Candy_Texture, nodeColor=<rgba 1 0.3281 0 1>}
  ) {samplerSource=SSM_Wrap_WorldGroupSettings, samplerType=SAMPLERTYPE_Masks}
  n149 = Multiply(A=@n148.R, B=LinearInterpolate(Alpha=@n148.G) {constA=0.2})
  n150 = NamedRerouteUsage()
  n151 = ScalarParameter() {defaultValue=-0.15, group=SuperLevel, parameterName=PatternRotate02}
  n152 = NamedRerouteUsage()
  n153 = TextureSample(
   UVs=Add(
    A=MaterialFunctionCall(Angle=@n151, UV/2DVector=@n152) {materialFunction=&MF_2DRotate}.Result
    B=@n4
   )
   Tex=NamedRerouteUsage()
  ) {samplerSource=SSM_Wrap_WorldGroupSettings, samplerType=SAMPLERTYPE_Masks}
  n154 = Add(A=Multiply(A=@n148.B, B=Multiply(A=@n153.G, B=@n153.G)), B=Multiply(A=@n153.R, B=@n153.R))
  n155 = Add(
   A=LinearInterpolate(B=ScalarParameter() {defaultValue=1.5, group=SuperLevel, parameterName=MaxBrightness01}, Alpha=Multiply(A=@n149, B=@n150))
   B=LinearInterpolate(B=ScalarParameter() {defaultValue=0.1, group=SuperLevel, parameterName=MaxBrightness02}, Alpha=Saturate(None=@n154))
  )
  n156 = NamedRerouteUsage()
  n157 = LinearInterpolate(Alpha=OneMinus(None=TextureSample(UVs=@n5.Result, Tex=@n156) {samplerSource=SSM_Wrap_WorldGroupSettings, samplerType=SAMPLERTYPE_Masks}.B)) {constA=0.5}
  n158 = LinearInterpolate(
   A=LinearInterpolate(B=ScalarParameter() {defaultValue=0.2, group=SuperLevel, parameterName=MaxBrightness03}, Alpha=Multiply(A=@n144.R, B=@n144.R))
   B=Saturate(None=Multiply(A=@n155, B=@n157))
   Alpha=NamedRerouteUsage()
  )
  NamedRerouteDeclaration(None=Multiply(A=LinearInterpolate(A=@n141, Alpha=@n158), B=ScalarParameter() {defaultValue=1, group=SuperLevel, parameterName=RoughnessIntensity})) {name=CandyRoughness, nodeColor=<rgba 0 0.625 1 1>}
  NamedRerouteDeclaration(None=Multiply(A=@n90, B=ScalarParameter() {defaultValue=2.2, group=SuperLevel, parameterName=UV_Tilling_Candy_Distort})) {name=UV_Candy_Distort, nodeColor=<rgba 1 0 0.7031 1>}
  n159 = Multiply(
   A=LinearInterpolate(A=ScalarParameter() {defaultValue=0.7, group=SuperLevel, parameterName=MinMetalic}, Alpha=@n158)
   B=ScalarParameter() {defaultValue=1, group=SuperLevel, parameterName=MetalicIntensity}
  )
  n160 = NamedRerouteUsage()
  NamedRerouteDeclaration(None=StaticSwitchParameter(True=Multiply(A=@n159, B=@n160), False=Constant()) {defaultValue=true, group=SuperLevel_Control, parameterName=UseSecondaryArea}) {name=MetalAdd, nodeColor=<rgba 1 0.5156 0 1>}
  NamedRerouteDeclaration(None=Multiply(A=@n3.R, B=@n3.R)) {name=NoiseChannel, nodeColor=<rgba 1 0.1641 0 1>}
  NamedRerouteDeclaration(None=@n31.B) {name=HairArea, nodeColor=<rgba 1 0.1172 0 1>}
  n161 = ScalarParameter() {defaultValue=0.5, group=SuperLevel, parameterName=UV_Tilling_Screen04}
  NamedRerouteDeclaration(None=Add(A=ComponentMask(None=@n135.RGB) {a=true, b=true}, B=Multiply(A=@n90, B=@n161))) {name=UV_Screen04, nodeColor=<rgba 0.8984 1 0 1>}
  n162 = TextureObjectParameter() {
   group=SuperLevel
   parameterName=Flowmap02
   samplerType=SAMPLERTYPE_LinearColor
   texture=&T_TorusFlowmap
  }
  n163 = TextureSample(Tex=@n162) {samplerType=SAMPLERTYPE_LinearColor, texture=&T_TorusFlowmap}
  NamedRerouteDeclaration(None=Multiply(A=ComponentMask(None=@n163.RGB) {g=true, r=true}, B=ScalarParameter() {defaultValue=0.1, parameterName=Distort_Pattern})) {name=Screen_Distort, nodeColor=<rgba 0 0.9297 1 1>}
  n164 = Custom() {code="return View.DirectionalLightDirection;", description=DirectionalLightVector, inputs []}
  n165 = DotProduct(A=@n164, B=@n164)
  n166 = CollectionParameter() {
   collection=&FortniteMaterialParameters
   parameterName=SunlightDirectionalVector
  }
  n167 = ComponentMask(None=@n166) {b=true, g=true, r=true}
  n168 = MaterialFunctionCall!(In=If(
    A=Custom() {code="return View.DirectionalLightColor;", description=DirectionalLightColor, inputs []}
    "A > B"=If(A=@n165, "A > B"=@n164, "A == B"=@n167, "A < B"=@n167) {desc=check if DirectionalLight exists}
    "A == B"=@n167
    "A < B"=@n167
   ) {desc=check if DirectionalLight color is Black}) {materialFunction=/Game/Packages/Fortress_SharedMaterials/Cinematics/Anime/MF_CINE_AnimeLightDirOverride}
  n169 = MaterialFunctionCall(Float3=@n168.Result) {materialFunction=&BreakOutFloat3Components}
  n170 = ScalarParameter() {
   desc="0.0 = the light vector is locked on the XY plane.\r\n1.0 = the actual light vector is used  \r\n(for Faces use a blend value of 0,0)"
   group=02_Anime Parameters
   parameterName=LockedZLightVector_Blend
   sliderMax=1
  }
  n171 = MaterialFunctionCall(X=@n169.R, Y=@n169.R, Z=Multiply(A=@n169.B, B=@n170)) {materialFunction=&MakeFloat3}
  n172 = StaticSwitchParameter(True=LinearInterpolate(A=@n168.Result, B=Normalize(VectorInput=@n171.Result), Alpha=VertexColor().A), False=@n168.Result) {
   desc="Lock the Z axis of the Light Vector\r\nis driven by the VertexColor A channel (0 = Regular LightVector, 1= Locked LIghtVector)"
   group=02_Anime Parameters
   parameterName=UseLockedZLightVector
   sortPriority=0
  }
  NamedRerouteDeclaration(None=@n172) {name=LightVector, nodeColor=<rgba 1 0.934 0.5103 1>}
  n173 = NamedRerouteUsage()
  n174 = ScalarParameter() {
   desc="LockedZ only for the specular highlight.\r\nfor a streched highlight in on the Z axis use the Blinn highlight"
   group=02_Anime Parameters
   parameterName=SpecularHighlight_LockedZVector
   sliderMax=1
  }
  n175 = AppendVector(A=Constant2Vector() {g=1, r=1}, B=Saturate(None=OneMinus(None=@n174)))
  n176 = Multiply(A=Add(A=CameraVectorWS(), B=@n173), B=@n175)
  n177 = Transform(None=NamedRerouteUsage())
  n178 = Multiply(A=@n173, B=@n175)
  n179 = StaticSwitchParameter(
   True=DotProduct(A=Normalize(VectorInput=@n176), B=@n177) {desc=Blinn}
   False=DotProduct(A=Normalize(VectorInput=@n178), B=ReflectionVectorWS(CustomWorldNormal=@n177)) {desc=Phong}
  ) {
   defaultValue=true
   desc="Default is the Phong highlight.\r\nUse Blinn for a vertical streched highlight"
   group=02_Anime Parameters
   parameterName=UseBlinnHighlight
   sortPriority=0
  }
  n180 = DotProduct(A=@n177, B=CameraVectorWS())
  n181 = Saturate(None=LinearInterpolate(
    A=@n179
    B=NamedRerouteDeclaration(None=@n180) {name=NdotV, nodeColor=<rgba 1 0.6242 0 1>}
    Alpha=ScalarParameter() {group=02_Anime Parameters, parameterName=SpecularHighlight_CameraFacing_Blend, sliderMax=1}
   ))
  n182 = Constant() {r=1}
  n183 = ScalarParameter() {defaultValue=0.3, group=02_Anime Parameters, parameterName=SpecularHighlight_Size, sliderMax=1, sliderMin=0.0001}
  n184 = ScalarParameter() {group=02_Anime Parameters, parameterName=SpecularHighlight_Hardness, sliderMax=1}
  NamedRerouteDeclaration(None=Multiply(
    A=SphereMask(A=@n181, B=@n182, Radius=@n183, Hardness=@n184) {hardnessPercent=78}
    B=ScalarParameter() {defaultValue=1, group=02_Anime Parameters, parameterName=SpecularHighlight_Intensity, sliderMax=8}
   )) {name=Specular_Curve, nodeColor=<rgba 1 0.4453 0 1>}
  n185 = Power(Base=Fresnel() {baseReflectFraction=0, exponent=1}, Exp=ScalarParameter() {defaultValue=0.4, group=SuperLevel, parameterName=Fresnel_DeFre}) {constExponent=0.5}
  n186 = Multiply(A=@n185, B=ScalarParameter() {defaultValue=2, group=SuperLevel, parameterName=MetalSine})
  n187 = Power(Base=ConstantBiasScale(None=Sine(None=@n186))) {constExponent=2.2}
  n188 = Power(Base=OneMinus(None=@n185), Exp=ScalarParameter() {defaultValue=0.5, group=SuperLevel, parameterName=FresnelPOW_Ref}) {constExponent=3}
  n189 = ScalarParameter() {defaultValue=0.5, group=SuperLevel, parameterName=Add_Basecolor}
  n190 = NamedRerouteUsage()
  n191 = Saturate(None=Add(A=Multiply(A=@n187, B=@n188), B=Multiply(A=@n189, B=@n190)))
  n192 = LinearInterpolate(
   A=ScalarParameter() {defaultValue=0.1, group=SuperLevel, parameterName=Min_BaseBaseColor}
   B=ScalarParameter() {defaultValue=0.5, group=SuperLevel, parameterName=Max_BaseBaseColor}
   Alpha=NamedRerouteDeclaration(None=@n158) {name=CandyPattern01, nodeColor=<rgba 0.2422 1 0 1>}
  )
  NamedRerouteDeclaration(None=Multiply(A=Multiply(A=@n191, B=@n192), B=ScalarParameter() {defaultValue=0.5, parameterName=BasecolorControl})) {name=CandyBaseColor_ORG, nodeColor=<rgba 1 0.9844 0 1>}
  NamedRerouteDeclaration(None=@n181) {name=LightingMask, nodeColor=<rgba 0 0.9766 1 1>}
  NamedRerouteDeclaration(None=@n96) {name=LineMask01, nodeColor=<rgba 1 0.0469 0 1>}
  NamedRerouteDeclaration(None=@n109) {name=LineMask02, nodeColor=<rgba 1 0.0469 0 1>}
  NamedRerouteDeclaration(None=@n118) {name=LineMask03, nodeColor=<rgba 1 0.0469 0 1>}
  n193 = Add(A=NamedRerouteUsage(), B=NamedRerouteUsage())
  n194 = NamedRerouteUsage()
  NamedRerouteDeclaration(None=Multiply(A=Add(A=@n193, B=@n194), B=NamedRerouteUsage())) {name=LineMask, nodeColor=<rgba 0.5938 0 1 1>}
  NamedRerouteDeclaration(None=ComponentMask(None=FunctionInput() {bUsePreviewValueAsDefault=true, inputName=M_Map, sortPriority=3, previewValue=<1 1 1 1>}) {b=true}) {name=Skin_Map, nodeColor=<rgba 0 0.2031 1 1>}
  NamedRerouteUsage()
  Power(Base=@n45) {constExponent=4}
  PreSkinnedNormal()
  PreSkinnedNormal()
  SkyAtmosphereLightDirection()
  Time()
 ]}
